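; Looping gameplay cue notify for the Chrono canopy item that, going by its name, plays while hiding.
; On looping start it takes the first spawned audio component and fades it in linearly over a short duration to full volume.
; The graph holds no other logic, and the construction script is empty.
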(event EventOnLoopingStart(EitherVFX-preferthisversion)
  (params (MyTarget :type Actor%20Object%20Reference) (Parameters :type Gameplay%20Cue%20Parameters%20Structure%20%28by%20ref%29) (ParticleComponents :type Array%20of%20FXSystem%20Component%20Object%20References) (AudioComponents :type Array%20of%20Audio%20Component%20Object%20References))
  (Audio|Components|Audio|FadeIn (Utilities|Array|Get(acopy) AudioComponents) 0.5 1.0 0.0 "Linear" :!class /Script/Engine.AudioComponent))

(fn ConstructionScript)
